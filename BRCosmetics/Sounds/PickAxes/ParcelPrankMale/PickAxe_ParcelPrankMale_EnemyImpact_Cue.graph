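# Sound cue for the ParcelPrankMale pickaxe hitting an enemy, using the pickaxe impact attenuation and impact sound class.
# A mixer layers one of five randomly chosen impact waves with a slightly delayed, quieter second layer.
# That second layer randomly picks with replacement between four laugh waves and repeated, heavier-weighted silence entries, so a laugh follows only about a quarter of hits.
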SoundCue /BRCosmetics/Sounds/PickAxes/ParcelPrankMale/PickAxe_ParcelPrankMale_EnemyImpact_Cue
root=Mixer_0
nodes [
 {id=Delay_0, children [Modulator_0], properties {delayMax=0.2, delayMin=0.1}}
 {id=Mixer_0, children [Random_0,Delay_0], properties {inputVolume [1,0.7]}}
 {id=Modulator_0, children [Random_1], properties {volumeMax=1, volumeMin=1}}
 {
  id=Random_0
  children [WavePlayer_0,WavePlayer_1,WavePlayer_2,WavePlayer_3,WavePlayer_4]
  properties {bShouldExcludeFromBranchCulling=true, weights [1,1,1,1,1]}
 }
 {
  id=Random_1
  children [WavePlayer_5,WavePlayer_6,WavePlayer_7,WavePlayer_8,WavePlayer_9,WavePlayer_9,WavePlayer_9,WavePlayer_9,WavePlayer_9]
  properties {bRandomizeWithoutReplacement=false, bShouldExcludeFromBranchCulling=true, weights [0.4,0.4,0.4,0.4,1,1,1,1,1]}
 }
 {id=WavePlayer_0, properties {soundWaveAssetPtr=&/PickAxe_ParcelPrankMale_EnemyImpact_01}}
 {id=WavePlayer_1, properties {soundWaveAssetPtr=&/PickAxe_ParcelPrankMale_EnemyImpact_02}}
 {id=WavePlayer_2, properties {soundWaveAssetPtr=&/PickAxe_ParcelPrankMale_EnemyImpact_03}}
 {id=WavePlayer_3, properties {soundWaveAssetPtr=&/PickAxe_ParcelPrankMale_EnemyImpact_05}}
 {id=WavePlayer_4, properties {soundWaveAssetPtr=&/PickAxe_ParcelPrankMale_EnemyImpact_06}}
 {id=WavePlayer_5, properties {soundWaveAssetPtr=&/PickAxe_ParcelPrankMale_EnemyImpact_Laughs_01}}
 {id=WavePlayer_6, properties {soundWaveAssetPtr=&/PickAxe_ParcelPrankMale_EnemyImpact_Laughs_02}}
 {id=WavePlayer_7, properties {soundWaveAssetPtr=&/PickAxe_ParcelPrankMale_EnemyImpact_Laughs_05}}
 {id=WavePlayer_8, properties {soundWaveAssetPtr=&/PickAxe_ParcelPrankMale_EnemyImpact_Laughs_03}}
 {id=WavePlayer_9, properties {soundWaveAssetPtr=&/1second_silence}}
]
properties {
 attenuationSettings=/Game/Sounds/Attenuation/Fort_Pickaxe_Impact_Att
 audioPropertiesSheet=/BRCosmetics/Sounds/Settings/PropertySheets/Pickaxe/APS_Cosmetic_Pickaxe_Impact
 bPrimeOnLoad=true
 maxDistance=4300
 soundClassObject=&Fort_Impacts
}
